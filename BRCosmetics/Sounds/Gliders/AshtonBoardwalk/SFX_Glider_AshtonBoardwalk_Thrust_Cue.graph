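# Sound cue for the thrust sound of the AshtonBoardwalk cosmetic glider.
# A single Random node at the root picks among five equally weighted ThrustIn wave variations.
# Playback uses reduced volume with the shared glider thrust attenuation and concurrency assets.
# It is assigned the glider Land sound class and the cosmetic glider Loop audio properties sheet, although no wave player sets looping.
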SoundCue /BRCosmetics/Sounds/Gliders/AshtonBoardwalk/SFX_Glider_AshtonBoardwalk_Thrust_Cue
root=Random_0
nodes [
 {id=Random_0, children [WavePlayer_0,WavePlayer_2,WavePlayer_3,WavePlayer_4,WavePlayer_5], properties {weights [1,1,1,1,1]}}
 {id=WavePlayer_0, properties {soundWaveAssetPtr=&/Glider_Ashton_Boardwalk_ThrustIn_01}}
 {id=WavePlayer_2, properties {soundWaveAssetPtr=&/Glider_Ashton_Boardwalk_ThrustIn_02}}
 {id=WavePlayer_3, properties {soundWaveAssetPtr=&/Glider_Ashton_Boardwalk_ThrustIn_03}}
 {id=WavePlayer_4, properties {soundWaveAssetPtr=&/Glider_Ashton_Boardwalk_ThrustIn_04}}
 {id=WavePlayer_5, properties {soundWaveAssetPtr=&/Glider_Ashton_Boardwalk_ThrustIn_05}}
]
properties {
 attenuationSettings=&Fort_Player_Glider_Thrust_Athena_Att
 audioPropertiesSheet=&APS_Cosmetic_Glider_Loop
 maxDistance=12500
 soundClassObject=&Fort_Gliders_Land
 volumeMultiplier=0.5
 concurrencySet [&Athena_Glider_Thrust_Conc]
}
